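; Building actor Blueprint for an outdoor inflatable palm prop that applies a material override from its building texture data.
; The construction script runs the parent's version, then, if the texture data array has more than one entry, reads the last entry's override material.
; If that material is valid, it goes onto slot 1 of the static mesh.
; BeginPlay, overlap and Tick events are empty.
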(event EventBeginPlay)

(event Collision|EventActorBeginOverlap
  (params (OtherActor :type Actor%20Object%20Reference)))

(event EventTick
  (params (DeltaSeconds :type Float%20%28single-precision%29)))

(fn ConstructionScript
  (|CallParentFunction)
  (bind _outtexturedataarray (Building|GetBuildingTextureData))
  (if (> (Utilities|Array|Length _outtexturedataarray) 1)
    (Utilities|IsValid (Variables|/Script/FortniteGame.BuildingTextureData|GetOverrideMaterial (Utilities|Array|Get(acopy) _outtexturedataarray (- (Utilities|Array|Length _outtexturedataarray) 1)))
      (:Is%20Valid
        (Rendering|Material|SetMaterial (Variables|Debug|GetStaticMeshComponent) 1 (Variables|/Script/FortniteGame.BuildingTextureData|GetOverrideMaterial (Utilities|Array|Get(acopy) _outtexturedataarray (- (Utilities|Array|Length _outtexturedataarray) 1))))))))
